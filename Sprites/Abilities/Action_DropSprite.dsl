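; Custom drop-item behavior for Sprites that spawns the dropped item as a projectile from the avatar pawn.
; HandleItemSpawn checks a rollback CVar for drops on elimination, spawns the projectile at a tuned up/forward offset, attaches the item entry, and returns false.
; LaunchProjectile sets its velocity from pawn velocity plus forward and upward impulses.
; Offsets, gravity and impulses come from scalable float rows, and spawn yaw is aim yaw rotated a quarter turn.
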(fn HandleItemSpawn
  (params (AvatarPawn :type Fort%20Pawn%20Object%20Reference) (Params :type Drop%20Item%20Custom%20Behavior%20Parameters%20Structure%20%28by%20ref%29))
  (bind (_returnvalue_roll _returnvalue_pitch _returnvalue_yaw) (Math|Rotator|BreakRotator (Pawn|GetBaseAimRotation (Variables|/Script/Engine.Actor|GetInstigator (Variables|GetAvatarPawn)))))
  (Utilities|FlowControl|Sequence
    (:then_0
      (if (not (> (Fortnite|GetNumericCVar "Fort.Rollback.DropSpritesOnElimination") 0))
        (Utilities|IsValid (Fortnite|GetFortPlayerControllerfromActor (Variables|GetAvatarPawn))
          (:Is%20Valid
            _)))
      (bind _returnvalue (Projectile|SpawnProjectile :ProjectileClass (Variables|Default|GetProjectile%20Class) :RequestedBy (Variables|GetAvatarPawn) :SpawnLocation (Utilities|Operators|Add (Transformation|GetActorLocation (Variables|GetAvatarPawn)) (* (Math|Vector|VectorUp) (ScalableFloat|GetValueatLevel (Variables|Default|GetRow_SpriteDropZOffset))) (* (Transformation|GetActorForwardVector (Variables|GetAvatarPawn)) (ScalableFloat|GetValueatLevel (Variables|Default|GetRow_SpriteDropForwardOffset)))) :SpawnRotation (Math|Rotator|MakeRotator _returnvalue_roll _returnvalue_pitch (+ _returnvalue_yaw 9e+01)) :SpawnDirection (Transformation|GetActorRotation (Variables|GetAvatarPawn)) :bUseDefaultPhysics false :InitialSpeed 0.0 :GravityScale (ScalableFloat|GetValueatLevel (Variables|Default|GetRow_SpriteDropGravity)) :ChargePercent 0.0))
      (Utilities|IsValid _returnvalue
        (:Is%20Valid
          (Class|FortItemEntryComponent|TryAddItemEntrytoActor (Utilities|Struct|BreakDropItemCustomBehaviorParameters (Variables|GetParams)) _returnvalue)
          (CallFunction|LaunchProjectile _returnvalue (Variables|GetAvatarPawn)))))
    (:then_1
      (return false))))

(fn LaunchProjectile
  (params (ProjectileActor :type Fort%20Projectile%20Base%20Object%20Reference) (AvatarPawn :type Fort%20Pawn%20Object%20Reference))
  (Utilities|IsValid AvatarPawn
    (:Is%20Valid
      (Projectile|SetVelocity (Variables|/Script/FortniteGame.FortProjectileBase|GetProjectileMovementComponent ProjectileActor) (Utilities|Operators|Add (* (ScalableFloat|GetValueatLevel (Variables|Default|GetRow_SpriteDropForwardImpulse)) (Transformation|GetActorForwardVector AvatarPawn)) (Transformation|GetVelocity AvatarPawn) (* (Math|Vector|VectorUp) (ScalableFloat|GetValueatLevel (Variables|Default|GetRow_SpriteDropZImpulse))))))))
